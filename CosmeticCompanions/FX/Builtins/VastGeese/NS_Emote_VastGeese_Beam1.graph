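# Niagara effect for the VastGeese cosmetic companion emote: a warm golden light beam that deploys, holds, then flickers near the end.
# Two cone-mesh emitters and a deploy mesh follow a sampled skeletal mesh; the cones' length also factors in a comparison with the owner skeleton.
# Rim_Glow adds pulsing ground sprites aligned by a slope trace, while Light_Down is disabled.
# Front-end modifiers adjust scale and brightness, and time-of-day selectors tint the glow.
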NiagaraSystem /CosmeticCompanions/FX/Builtins/VastGeese/NS_Emote_VastGeese_Beam1
dependencies {
 usedDataInterfaces [
  &NiagaraDataInterfaceCurve
  &NiagaraDataInterfaceVectorCurve
  &NiagaraDataInterfaceColorCurve
  &NiagaraDataInterfaceCollisionQuery
  &NiagaraDataInterfaceDebugDraw
 ]
 usedDynamicInputs [&ColorFromCurve, &Multiply_Vector2D_ByFloat]
 usedModules [
  &SystemState
  &self:SystemSpawnScript.SystemScriptSource.SystemScriptGraph.NiagaraNodeAssignment_0.SetVariables_#1
 ]
}
emitters {
 ConeLights {
  emitterSpawnScript [/SetVariables(
    Emitter.FADE POWER:NiagaraFloat=&FN_FrontEndFloatModifier(Gameplay=0.05, Front End=0)
    Emitter.FADE SOFTNESS=0.4
    Emitter.FRONT-END FADE=1
   ) {isSetParameters=true}]
  emitterUpdateScript [
   &EmitterState(Life Cycle Mode=::NewEnumerator1, Loop Behavior=::NewEnumerator1, Loop Delay=1.7, Loop Duration=4, UseLoopDelay=false)
   &SpawnBurst_Instantaneous(Use Spawn Probability=false, UseLoopCountLimit=false)
  ]
  particleSpawnScript [
   &InitializeParticle(
    Color=<rgba 1 0.5859 0 0.05>
    Lifetime=4.6
    Mesh Scale=<0.6 0.6 0.6>
    Mesh Scale Mode=::NewEnumerator4
    Sprite Size Mode=::NewEnumerator3
    Uniform Sprite Size=1000
    UsePositionOffset=false
    Write Material Random=true
   )
   &InitialMeshOrientation(Mesh Orientation Mode=::NewEnumerator2, Use Rotation Vector=true)
  ]
  particleUpdateScript [
   &SolveForcesAndVelocity(Clamp Velocity=false, Limit Acceleration=false)
   &ParticleState()
   &SkeletalMeshLocation(
    MaskSpawnGroup=false
    Mesh Sampling Type=::NewEnumerator5
    Orientation Sampling=::NewEnumerator1
    Sampled Position Offset=<0 0 -5>
    Skeletal Mesh=$System.Skeletal Mesh
    Socket Sampling Mode=::NewEnumerator5
   ) {module=SkeletalMeshLocation001}
   &SkeletalMeshLocation(
    MaskSpawnGroup=false
    Mesh Sampling Type=::NewEnumerator5
    Orientation Sampling=::NewEnumerator1
    Position Sampling=::NewEnumerator2
    Skeletal Mesh=$User.OwnerSkeleton
    Socket Sampling Mode=::NewEnumerator5
   ) {module=SkeletalMeshLocation002}
   &ScaleMeshSize(Scale Factor=&FN_FrontEndVector3Modifier(
     Gameplay=&Multiply_Vector(
      A=&MakeVector(
       X=&FloatFromCurve(FloatCurve=DataInterface() {propertyValues {Curve {keys [
            <at 0 0 RCIM_Cubic>
            {arriveTangentWeight=0.0247, interpMode=RCIM_Cubic, leaveTangentWeight=0.0255, tangentMode=RCTM_User, time=0.0174, value=0.05}
            {arriveTangentWeight=0.0589, interpMode=RCIM_Cubic, leaveTangentWeight=0.2882, tangentMode=RCTM_User, time=0.0565, value=0.05}
            {
             arriveTangent=3.9696
             arriveTangentWeight=0.0453
             interpMode=RCIM_Cubic
             leaveTangent=3.9696
             leaveTangentWeight=0.4261
             tangentMode=RCTM_User
             time=0.1319
             value=0.75
            }
            {arriveTangent=0.288, interpMode=RCIM_Cubic, leaveTangent=0.288, time=0.2388, value=0.8}
            {arriveTangentWeight=0.3013, interpMode=RCIM_Cubic, tangentMode=RCTM_User, time=1, value=1}
           ]}}})
       Y=&FloatFromCurve(FloatCurve=DataInterface() {propertyValues {Curve {keys [
            <at 0 0 RCIM_Cubic>
            {arriveTangentWeight=0.0247, interpMode=RCIM_Cubic, leaveTangentWeight=0.0255, tangentMode=RCTM_User, time=0.0174, value=0.05}
            {arriveTangentWeight=0.0589, interpMode=RCIM_Cubic, leaveTangentWeight=0.2882, tangentMode=RCTM_User, time=0.0565, value=0.05}
            {
             arriveTangent=3.9696
             arriveTangentWeight=0.0453
             interpMode=RCIM_Cubic
             leaveTangent=3.9696
             leaveTangentWeight=0.4261
             tangentMode=RCTM_User
             time=0.1319
             value=0.75
            }
            {arriveTangent=0.288, interpMode=RCIM_Cubic, leaveTangent=0.288, time=0.2254, value=0.8}
            {arriveTangentWeight=0.3013, interpMode=RCIM_Cubic, tangentMode=RCTM_User, time=1, value=1}
           ]}}})
       Z=&Multiply_Float(
        A=&FloatFromCurve(FloatCurve=DataInterface() {propertyValues {Curve {keys [
             {arriveTangent=779.4893, interpMode=RCIM_Cubic, leaveTangent=779.4893, leaveTangentWeight=3.3909, tangentMode=RCTM_User, time=0, value=0}
             {
              arriveTangent=19.032
              arriveTangentWeight=2.8007
              interpMode=RCIM_Cubic
              leaveTangent=19.032
              leaveTangentWeight=13.8693
              tangentMode=RCTM_User
              time=0.099
              value=30
             }
             {arriveTangentWeight=0.2773, interpMode=RCIM_Cubic, leaveTangentWeight=0.3745, tangentMode=RCTM_User, time=1.2045, value=45}
            ]}}})
        B=&RemapRange(
         Input Value To Remap=&VectorLength(Vector=&Subtract_Position(
           A=Linked() {linkedVariable {name=Particles.SkeletalMeshLocation001.SampledScale, type=Vector3f}}
           B=Linked() {linkedVariable {name=Particles.SkeletalMeshLocation002.SampledScale, type=Vector3f}}
          ))
         Input Min=0
         Input Max=100
         Output Min=0
         Output Max=2.5
        )
       )
      )
      B=&Divide_Vector(A:Vector3f=<1 1 1>, B:Vector3f=$Engine.Owner.Scale)
     )
     Front End=<0.8 0.8 1.2>
    ))
   &ScaleColor(
    Scale Alpha=&FloatFromCurve(
     FloatCurve=DataInterface() {propertyValues {Curve {keys [
         {interpMode=RCIM_Cubic, leaveTangent=0.0402, tangentMode=RCTM_Break, time=0, value=0}
         <at 0.3 0.0075>
         {arriveTangent=-0.3583, leaveTangent=-0.3583, time=3.9058, value=0.0075}
         <at 3.9074 0.0005>
         <at 3.9555 0.0005>
         <at 3.9764 0.0075>
         <at 4.1455 0.0075>
         <at 4.1461 0.0005>
         <at 4.2511 0.0005>
         <at 4.2699 0.0075>
         <at 4.3362 0.0075>
         <at 4.3431 0.0005>
         <at 4.4684 0.0005>
         <at 4.497 0.0075>
        ]}}}
     CurveIndex=$System.Age
     Scale Curve=&FN_FrontEndFloatModifier(Gameplay=1, Front End=0.5)
    )
    ScaleA=true
    ScaleRGB=true
   )
   /SetVariables(Particles.TechnoBeamColor:LinearColor=&MultiplyLinearColor(
     Linear Color=<rgba 13 5.695 0.7448 0.003>
     Scale Factor=&MakeVector4(X=1, Y=1, Z=1, W=&FN_FrontEndFloatModifier(Gameplay=1, Front End=0.5))
    )) {isSetParameters=true, module=SetVariables_#2}
   &FN_Light_Attributes(
    Color=$Particles.Color
    Create Color Attribute=true
    Create Exponent Attribute=true
    Create Position Attribute=true
    Create Radius Attribute=true
    Exponent=2
    Position=&AddVectorToPosition(
     Position=$Particles.Position
     Vector=&VectorFromCurve(
      VectorCurve=DataInterface() {propertyValues {XCurve {keys [<at 0 0>,<at 1 0>]}, YCurve {keys [<at 0 0>,<at 1 0>]}, ZCurve {keys [<at 0 1>,<at 0.65 1>,<at 1 0.7>]}}}
      Scale Curve=<1 1 -20>
     )
    )
    Radius=90
    Scale Color=&FloatFromCurve(
     FloatCurve=DataInterface() {propertyValues {Curve {keys [<at 0 0>,<at 0.16 1>,<at 0.65 1>,<at 1 0>]}}}
     Scale Curve=&FN_FrontEndFloatModifier(Gameplay=100, Front End=20)
    )
   )
   /SetVariables(Particles.DynamicMaterialParameterTechBeam:Vector4f=/Niagara/DynamicInputs/Multiply/Multiply_Vector4_ByFloat(
     Vector4:Vector4f=<0.8 0.8 1 1>
     Float:NiagaraFloat=&FloatFromCurve(
      FloatCurve=DataInterface() {propertyValues {Curve {keys [
          {interpMode=RCIM_Cubic, leaveTangent=1.6164, tangentMode=RCTM_Break, time=0, value=0}
          {arriveTangentWeight=0.173, interpMode=RCIM_Cubic, leaveTangentWeight=1.0013, tangentMode=RCTM_User, time=0.4503, value=1}
          {arriveTangent=-0.0121, leaveTangent=-0.3583, tangentMode=RCTM_User, time=4.0562, value=1}
          <at 4.0574 0.25>
          <at 4.1055 0.25>
          <at 4.1267 1>
          <at 4.2958 1>
          <at 4.2961 0.25>
          <at 4.4011 0.25>
          <at 4.4202 1>
          <at 4.4865 1>
          <at 4.4931 0.25>
          <at 4.6184 0.25>
          <at 4.6473 1>
         ]}}}
      CurveIndex=$Engine.System.Age
     )
    )) {isSetParameters=true, module=SetVariables_#3}
   &DynamicMaterialParameters(
    Index 0 Param 4=&FloatFromCurve(
     FloatCurve=DataInterface() {propertyValues {Curve {keys [
         {interpMode=RCIM_Cubic, leaveTangent=1.6164, tangentMode=RCTM_Break, time=0, value=0}
         {arriveTangentWeight=0.173, interpMode=RCIM_Cubic, leaveTangentWeight=1.0013, tangentMode=RCTM_User, time=0.4503, value=1}
         {arriveTangent=-0.0121, leaveTangent=-0.3583, tangentMode=RCTM_User, time=4.0562, value=1}
         <at 4.0574 0.25>
         <at 4.1055 0.25>
         <at 4.1267 1>
         <at 4.2958 1>
         <at 4.2961 0.25>
         <at 4.4011 0.25>
         <at 4.4202 1>
         <at 4.4865 1>
         <at 4.4931 0.25>
         <at 4.6184 0.25>
         <at 4.6473 1>
        ]}}}
     CurveIndex=$Engine.System.Age
     Scale Curve=&FN_FrontEndFloatModifier(Gameplay=1, Front End=0.5)
    )
    Index 0 Vector 1:Vector3f=&FN_Select_Vector_from_Time_Of_Day(
     Morning=&MakeVectorFromLinearColorRGB(Color=<rgba 3 2.4842 0.7221 1>)
     Day=&MakeVectorFromLinearColorRGB(Color=<rgba 3 2.4842 0.7221 1>)
     Evening=&MakeVectorFromLinearColorRGB(Color=<rgba 3 2.4842 0.7221 1>)
     Night=&MakeVectorFromLinearColorRGB(Color=<rgba 5 2.6977 0.5729 1>)
    )
    Module.Index0Float3Write=true
    Module.Index0VectorWrite=true
    Write Parameter Index 0=::NewEnumerator3
   )
  ]
  rendererClasses [&NiagaraLightRendererProperties, &NiagaraMeshRendererProperties]
  renderers [
   {rendererClass=&NiagaraLightRendererProperties, rendererIndex=0}
   {rendererClass=&NiagaraMeshRendererProperties, rendererIndex=1}
   {rendererClass=&NiagaraMeshRendererProperties, rendererIndex=2}
  ]
 }
 ConeLights001 {
  emitterSpawnScript [/SetVariables(
    Emitter.FADE POWER:NiagaraFloat=&FN_FrontEndFloatModifier(Gameplay=0.8, Front End=1.1)
    Emitter.FLIP DIRECTION=1
    Emitter.FRONT-END FADE=1
   ) {isSetParameters=true}]
  emitterUpdateScript [
   &EmitterState(Life Cycle Mode=::NewEnumerator1, Loop Behavior=::NewEnumerator1, Loop Delay=1.7, Loop Duration=4, UseLoopDelay=false)
   &SpawnBurst_Instantaneous(Use Spawn Probability=false, UseLoopCountLimit=false)
  ]
  particleSpawnScript [
   &InitializeParticle(
    Color=<rgba 1.5 1.1715 0.633 0.35>
    Lifetime=4.6
    Mesh Scale=<0.6 0.6 0.6>
    Mesh Scale Mode=::NewEnumerator4
    Sprite Size Mode=::NewEnumerator3
    Uniform Sprite Size=1000
    UsePositionOffset=false
    Write Material Random=true
   )
   &InitialMeshOrientation(
    Mesh Axis To Orient=<0 0 1>
    Mesh Orientation Mode=::NewEnumerator2
    Orientation Vector=<0 0 1>
    Rotation=<0 0 0.5>
    Use Rotation Vector=true
   )
  ]
  particleUpdateScript [
   &SolveForcesAndVelocity(Clamp Velocity=false, Limit Acceleration=false)
   &ParticleState()
   &ScaleMeshSize(Scale Factor=<0.98 0.98 1>)
   &SkeletalMeshLocation(
    MaskSpawnGroup=false
    Mesh Sampling Type=::NewEnumerator5
    Orientation Sampling=::NewEnumerator1
    Sampled Position Offset=<0 0 -5>
    Skeletal Mesh=$System.Skeletal Mesh
    Socket Sampling Mode=::NewEnumerator5
   ) {module=SkeletalMeshLocation001}
   &SkeletalMeshLocation(
    MaskSpawnGroup=false
    Mesh Sampling Type=::NewEnumerator5
    Orientation Sampling=::NewEnumerator1
    Position Sampling=::NewEnumerator2
    Skeletal Mesh=$User.OwnerSkeleton
    Socket Sampling Mode=::NewEnumerator5
   ) {module=SkeletalMeshLocation002}
   &ScaleMeshSize(Scale Factor=&FN_FrontEndVector3Modifier(
     Gameplay=&Multiply_Vector(
      A=&MakeVector(
       X=&FloatFromCurve(FloatCurve=DataInterface() {propertyValues {Curve {keys [
            <at 0 0 RCIM_Cubic>
            {arriveTangentWeight=0.0247, interpMode=RCIM_Cubic, leaveTangentWeight=0.0255, tangentMode=RCTM_User, time=0.0174, value=0.05}
            {arriveTangentWeight=0.0589, interpMode=RCIM_Cubic, leaveTangentWeight=0.2882, tangentMode=RCTM_User, time=0.0565, value=0.05}
            {
             arriveTangent=3.9696
             arriveTangentWeight=0.0453
             interpMode=RCIM_Cubic
             leaveTangent=3.9696
             leaveTangentWeight=0.4261
             tangentMode=RCTM_User
             time=0.1319
             value=0.75
            }
            {arriveTangent=0.0576, interpMode=RCIM_Cubic, leaveTangent=0.0576, time=0.2388, value=0.8}
            {arriveTangentWeight=0.3013, interpMode=RCIM_Cubic, tangentMode=RCTM_User, time=1, value=0.8}
           ]}}})
       Y=&FloatFromCurve(FloatCurve=DataInterface() {propertyValues {Curve {keys [
            <at 0 0 RCIM_Cubic>
            {arriveTangentWeight=0.0247, interpMode=RCIM_Cubic, leaveTangentWeight=0.0255, tangentMode=RCTM_User, time=0.0174, value=0.05}
            {arriveTangentWeight=0.0589, interpMode=RCIM_Cubic, leaveTangentWeight=0.2882, tangentMode=RCTM_User, time=0.0565, value=0.05}
            {
             arriveTangent=3.9696
             arriveTangentWeight=0.0453
             interpMode=RCIM_Cubic
             leaveTangent=3.9696
             leaveTangentWeight=0.4261
             tangentMode=RCTM_User
             time=0.1319
             value=0.75
            }
            {arriveTangent=0.0576, interpMode=RCIM_Cubic, leaveTangent=0.0576, time=0.2254, value=0.8}
            {arriveTangentWeight=0.3013, interpMode=RCIM_Cubic, tangentMode=RCTM_User, time=1, value=0.8}
           ]}}})
       Z=&Multiply_Float(
        A=&FloatFromCurve(FloatCurve=DataInterface() {propertyValues {Curve {keys [
             {arriveTangent=779.4893, interpMode=RCIM_Cubic, leaveTangent=779.4893, leaveTangentWeight=3.3909, tangentMode=RCTM_User, time=0, value=0}
             {
              arriveTangent=19.032
              arriveTangentWeight=2.8007
              interpMode=RCIM_Cubic
              leaveTangent=19.032
              leaveTangentWeight=13.8693
              tangentMode=RCTM_User
              time=0.099
              value=30
             }
             {arriveTangentWeight=0.2773, interpMode=RCIM_Cubic, leaveTangentWeight=0.3745, tangentMode=RCTM_User, time=1.2045, value=35}
            ]}}})
        B=&RemapRange(
         Input Value To Remap=&VectorLength(Vector=&Subtract_Position(
           A=Linked() {linkedVariable {name=Particles.SkeletalMeshLocation001.SampledScale, type=Vector3f}}
           B=Linked() {linkedVariable {name=Particles.SkeletalMeshLocation002.SampledScale, type=Vector3f}}
          ))
         Input Min=0
         Input Max=100
         Output Min=0
         Output Max=2.5
        )
       )
      )
      B=&Divide_Vector(A:Vector3f=<1 1 1>, B:Vector3f=$Engine.Owner.Scale)
     )
     Front End=<0.8 0.8 1.2>
    )) {module=ScaleMeshSize001}
   &ScaleColor(
    Scale Alpha=&FloatFromCurve(
     FloatCurve=DataInterface() {propertyValues {Curve {keys [
         {interpMode=RCIM_Cubic, leaveTangent=1.6164, tangentMode=RCTM_Break, time=0, value=0}
         {arriveTangentWeight=0.173, interpMode=RCIM_Cubic, leaveTangentWeight=1.0013, tangentMode=RCTM_User, time=0.4503, value=1}
         {arriveTangent=-0.0121, leaveTangent=-0.3583, tangentMode=RCTM_User, time=4.0562, value=1}
         <at 4.0574 0.25>
         <at 4.1055 0.25>
         <at 4.1267 1>
         <at 4.2958 1>
         <at 4.2961 0.25>
         <at 4.4011 0.25>
         <at 4.4202 1>
         <at 4.4865 1>
         <at 4.4931 0.25>
         <at 4.6184 0.25>
         <at 4.6473 1>
        ]}}}
     CurveIndex=$Engine.System.Age
     Scale Curve=&Multiply_Float(A=0.6, B=&FN_FrontEndFloatModifier(Gameplay=1, Front End=0.5))
    )
    ScaleA=true
    ScaleRGB=true
   )
   /SetVariables(Particles.TechnoBeamColor:LinearColor=&MultiplyLinearColor(Linear Color=<rgba 1.5 1.1715 0.633 0.35>, Scale Factor=&MakeVector4(X=1, Y=1, Z=1, W=1))) {enabled=false, isSetParameters=true, module=SetVariables_#2}
   &FN_Light_Attributes(
    Color=$Particles.Color
    Create Color Attribute=true
    Create Exponent Attribute=true
    Create Position Attribute=true
    Create Radius Attribute=true
    Exponent=4
    Position=&AddVectorToPosition(Position=$Particles.Position, Vector=<0 0 -10>)
    Radius=90
    Scale Color=10
   )
   /SetVariables(Particles.DynamicMaterialParameterTechBeam:Vector4f=&Multiply_Vector4_ByFloat(
     Vector4:Vector4f=<0.8 0.8 1 1>
     Float:NiagaraFloat=&FloatFromCurve(
      FloatCurve=DataInterface() {propertyValues {Curve {keys [
          {interpMode=RCIM_Cubic, leaveTangent=1.6164, tangentMode=RCTM_Break, time=0, value=0}
          {arriveTangentWeight=0.173, interpMode=RCIM_Cubic, leaveTangentWeight=1.0013, tangentMode=RCTM_User, time=0.4503, value=1}
          {arriveTangent=-0.0121, leaveTangent=-0.3583, tangentMode=RCTM_User, time=4.0562, value=1}
          <at 4.0574 0.25>
          <at 4.1055 0.25>
          <at 4.1267 1>
          <at 4.2958 1>
          <at 4.2961 0.25>
          <at 4.4011 0.25>
          <at 4.4202 1>
          <at 4.4865 1>
          <at 4.4931 0.25>
          <at 4.6184 0.25>
          <at 4.6473 1>
         ]}}}
      CurveIndex=$Engine.System.Age
     )
    )) {isSetParameters=true, module=SetVariables_#3}
   &DynamicMaterialParameters(
    Index 0 Param 4=&FloatFromCurve(
     FloatCurve=DataInterface() {propertyValues {Curve {keys [
         {interpMode=RCIM_Cubic, leaveTangent=1.6164, tangentMode=RCTM_Break, time=0, value=0}
         {arriveTangentWeight=0.173, interpMode=RCIM_Cubic, leaveTangentWeight=1.0013, tangentMode=RCTM_User, time=0.4503, value=1}
         {arriveTangent=-0.0121, leaveTangent=-0.3583, tangentMode=RCTM_User, time=4.0562, value=1}
         <at 4.0574 0.25>
         <at 4.1055 0.25>
         <at 4.1267 1>
         <at 4.2958 1>
         <at 4.2961 0.25>
         <at 4.4011 0.25>
         <at 4.4202 1>
         <at 4.4865 1>
         <at 4.4931 0.25>
         <at 4.6184 0.25>
         <at 4.6473 1>
        ]}}}
     CurveIndex=$Engine.System.Age
     Scale Curve=&FN_FrontEndFloatModifier(Gameplay=1, Front End=0.5)
    )
    Index 0 Vector 1:Vector3f=&FN_Select_Vector_from_Time_Of_Day(
     Morning=&MakeVectorFromLinearColorRGB(Color=<rgba 3 2.4842 0.7221 1>)
     Day=&MakeVectorFromLinearColorRGB(Color=<rgba 3 2.4842 0.7221 1>)
     Evening=&MakeVectorFromLinearColorRGB(Color=<rgba 3 2.4842 0.7221 1>)
     Night=&MakeVectorFromLinearColorRGB(Color=<rgba 5 2.6977 0.5729 1>)
    )
    Module.Index0Float3Write=true
    Module.Index0VectorWrite=true
    Write Parameter Index 0=::NewEnumerator3
   )
  ]
  rendererClasses [&NiagaraMeshRendererProperties]
  renderers [{rendererClass=&NiagaraMeshRendererProperties, rendererIndex=0}]
 }
 DeployConeLight {
  emitterSpawnScript [/SetVariables(
    Emitter.FADE POWER:NiagaraFloat=&FN_FrontEndFloatModifier(Gameplay=0.8, Front End=1.1)
    Emitter.FRONT-END FADE=1
   ) {enabled=false, isSetParameters=true}]
  emitterUpdateScript [
   &EmitterState(Life Cycle Mode=::NewEnumerator1, Loop Behavior=::NewEnumerator1, Loop Delay=1.6, Loop Duration=4, UseLoopDelay=false)
   &SpawnBurst_Instantaneous(Use Spawn Probability=false, UseLoopCountLimit=false)
  ]
  particleSpawnScript [
   &InitializeParticle(
    Color=<rgba 1 0.5316 0 0.7>
    Lifetime=4.6
    Mesh Scale=<0.8 0.8 0.8>
    Mesh Scale Mode=::NewEnumerator4
    Sprite Size Mode=::NewEnumerator1
    Uniform Sprite Size Max=15
    Uniform Sprite Size Min=10
    UsePositionOffset=false
    Write Material Random=true
   )
   /SetVariables(Particles.TinyRingsScale:Vector3f=&VectorFromFloat(Value=0.3)) {isSetParameters=true}
   /SetVariables(Particles.TinyRingsOrientation:Quat4f=&MakeQuaternion(Quaternion From=1, Axis=<1 0 -90>, Angle=0, Roll=180, From=?, To=?, Angle Type=::NewEnumerator0, Normalize Vectors=false)) {isSetParameters=true}
   /SetVariables(Particles.Linear Color=&MultiplyLinearColor(
     Linear Color=<rgba 2 0.9334 0 1>
     Scale Factor=&MakeVector4(X=1, Y=1, Z=1, W=&FN_FrontEndFloatModifier(Gameplay=1, Front End=0.5))
    )) {isSetParameters=true}
  ]
  particleUpdateScript [
   &SolveForcesAndVelocity(Clamp Velocity=false, Limit Acceleration=false)
   &ParticleState()
   &ScaleMeshSize(Scale Factor=&VectorFromCurve(VectorCurve=DataInterface() {propertyValues {
       XCurve {keys [
         {interpMode=RCIM_Cubic, leaveTangentWeight=0.0184, tangentMode=RCTM_User, time=0, value=0.1984}
         {arriveTangent=0.0461, interpMode=RCIM_Cubic, leaveTangent=0.0461, tangentMode=RCTM_User, time=0.0553, value=0.2}
         {arriveTangentWeight=0.0411, interpMode=RCIM_Cubic, leaveTangentWeight=0.3716, tangentMode=RCTM_User, time=0.1552, value=1}
         {arriveTangentWeight=0.2836, interpMode=RCIM_Cubic, tangentMode=RCTM_User, time=1, value=1}
        ]}
       YCurve {keys [
         {interpMode=RCIM_Cubic, leaveTangentWeight=0.0182, tangentMode=RCTM_User, time=0, value=0.2}
         {arriveTangent=0.0455, interpMode=RCIM_Cubic, leaveTangent=0.0455, tangentMode=RCTM_User, time=0.0545, value=0.2}
         {arriveTangentWeight=0.0414, interpMode=RCIM_Cubic, leaveTangentWeight=0.3714, tangentMode=RCTM_User, time=0.1552, value=1}
         {arriveTangentWeight=0.2836, interpMode=RCIM_Cubic, tangentMode=RCTM_User, time=1, value=1}
        ]}
       ZCurve {keys [
         {arriveTangent=4.252, interpMode=RCIM_Cubic, leaveTangent=4.252, tangentMode=RCTM_User, time=0, value=0}
         {arriveTangent=1.2037, interpMode=RCIM_Cubic, leaveTangent=1.2037, tangentMode=RCTM_User, time=0.1857, value=0.8}
         {arriveTangent=0.1842, interpMode=RCIM_Cubic, leaveTangent=0.1842, time=0.2993, value=0.9}
         {arriveTangentWeight=0.2759, interpMode=RCIM_Cubic, tangentMode=RCTM_User, time=1, value=0.95}
        ]}
      }}))
   &SkeletalMeshLocation(
    MaskSpawnGroup=false
    Mesh Sampling Type=::NewEnumerator5
    Orientation Sampling=::NewEnumerator1
    Sampled Position Offset=<0 0 -3>
    Skeletal Mesh=$System.Skeletal Mesh
    Socket Sampling Mode=::NewEnumerator5
   ) {module=SkeletalMeshLocation001}
   &ScaleColor(
    Scale Alpha=&FloatFromCurve(
     FloatCurve=DataInterface() {propertyValues {Curve {keys [
         {interpMode=RCIM_Cubic, leaveTangent=1.6164, tangentMode=RCTM_Break, time=0, value=0.1608}
         <at 0.3 0.7>
         {arriveTangent=-0.3583, leaveTangent=-0.3583, time=3.9058, value=0.7}
         <at 3.9221 0.2>
         <at 3.9702 0.2>
         <at 3.9764 0.7>
         <at 4.1455 0.7>
         <at 4.1609 0.2>
         <at 4.2658 0.2>
         <at 4.2699 0.7>
         <at 4.3362 0.7>
         <at 4.3578 0.2>
         <at 4.4831 0.2>
         <at 4.497 0.7>
        ]}}}
     CurveIndex=$Engine.System.Age
     Scale Curve=&FN_FrontEndFloatModifier(Gameplay=1, Front End=0.5)
    )
    ScaleA=true
    ScaleRGB=true
   )
  ]
  rendererClasses [&NiagaraMeshRendererProperties]
  renderers [
   {rendererClass=&NiagaraMeshRendererProperties, rendererIndex=0}
   {rendererClass=&NiagaraMeshRendererProperties, rendererIndex=1}
   {rendererClass=&NiagaraMeshRendererProperties, rendererIndex=2}
  ]
 }
 Light_Down {
  bEnabled=false
  emitterUpdateScript [
   &EmitterState(
    Inactive Response=::NewEnumerator1
    Life Cycle Mode=::NewEnumerator1
    Loop Behavior=::NewEnumerator1
    Loop Delay=0.2
    Loop Duration Mode=::NewEnumerator1
    UseLoopDelay=true
   )
   &SpawnBurst_Instantaneous(Use Spawn Probability=false, UseLoopCountLimit=false)
  ]
  particleSpawnScript [&InitializeParticle(
    Color=<rgba 1.276 3.8972 5 1>
    Lifetime=4.6
    Mass Max=1.5
    Mass Mode=::NewEnumerator3
    Position Offset=<0 0 150>
    Sprite Rotation Mode=::NewEnumerator2
    Sprite Size Mode=::NewEnumerator1
    Uniform Sprite Size Max=180
    Uniform Sprite Size Min=150
    UsePositionOffset=true
    Write Material Random=true
   )]
  particleUpdateScript [
   &ParticleState()
   &FN_Light_Attributes(
    Color=<rgba 1 0.6259 0.2267 1>
    Create Color Attribute=true
    Create Exponent Attribute=true
    Create Radius Attribute=true
    Exponent=1
    Radius=0.1
    Scale Color=&FNTimeOfDayScalar(Night Scale=1, Morning Scale=1, Evening Scale=1)
   )
   &FN_Gen5_Light(
    Attenutation:NiagaraFloat=&FloatFromCurve(FloatCurve=DataInterface() {propertyValues {Curve {keys [<at 0 1>,<at 0.62 1>,<at 1 0.3>]}}}, Scale Curve=180)
    Color:LinearColor=<rgba 1 0.4455 0.0034 1>
    Exponent=6
    Intensity:NiagaraFloat=&FloatFromCurve(FloatCurve=DataInterface() {propertyValues {Curve {keys [<at 0 1>,<at 0.5822 1>,<at 1 0>]}}}, Scale Curve=55)
   )
  ]
  rendererClasses [&NiagaraComponentRendererProperties]
  renderers [{rendererClass=&NiagaraComponentRendererProperties, rendererIndex=0}]
 }
 Rim_Glow {
  emitterUpdateScript [
   &EmitterState(Life Cycle Mode=::NewEnumerator1, Loop Behavior=::NewEnumerator1, Loop Duration=4.2, UseLoopDelay=false)
   &SpawnRate(SpawnRate=4, Use Spawn Probability=false)
   &SpawnBurst_Instantaneous(Spawn Count=1, Use Spawn Probability=false) {enabled=false}
   /SetVariables(Emitter.FE_Toggle:NiagaraBool=&Boolean_NotOperation(Bool=&Boolean_OrOperation(Bool1=$User.bIsFrontEnd, Bool2=$User.bIsFrontEndPreview))) {isSetParameters=true}
  ]
  particleSpawnScript [
   &InitializeParticle(
    Color=&FN_Select_Color_from_Time_Of_Day(
     Morning=<rgba 0.1348 0.0876 0.0187 0.5435>
     Day=<rgba 0.1348 0.0876 0.0187 0.5435>
     Evening=<rgba 0.1348 0.0876 0.0187 0.5435>
     Night=<rgba 0.1348 0.1098 0.0828 0.5435>
    )
    Mass Mode=::NewEnumerator1
    Sprite Size Mode=::NewEnumerator3
    Uniform Sprite Size=320
    UsePositionOffset=false
    Write Material Random=true
   ) {module=InitializeParticle001}
   &CameraOffset(Camera Offset Amount=3) {enabled=false}
   &SpriteFacingAndAlignment(Sprite Facing=<0 0 1>, WriteAlignment=false, WriteFacing=true)
  ]
  particleUpdateScript [
   &ParticleState()
   &ScaleSpriteSize(
    Uniform Curve Scale=0.9
    Uniform Curve Sprite Scale=DataInterface() {propertyValues {Curve {keys [{tangentMode=RCTM_User, time=0, value=0.5118}, {tangentMode=RCTM_User, time=1, value=1}]}}}
   )
   &FN_Light_v3(
    Color=<rgba 1 0.325 0.1 1>
    Create Color Attribute=true
    Create Exponent Attribute=true
    Create Radius Attribute=true
    Exponent=8
    Radius=120
    Scale Color=10
   )
   &self:Slope_Position_05(
    Apply Position=true
    Apply Rotation=true
    Height Adjustment/Offset=1
    Slope Position:NiagaraPosition=$Engine.Emitter.SimulationPosition
    Trace Length=120
    Trace Origin Height=85
    bIsFrontEnd:NiagaraBool=$User.bIsFrontEndPreview
   ) {module=Slope_Position_05}
   &ScaleColor(
    Scale Alpha=&FloatFromCurve(
     FloatCurve=DataInterface() {propertyValues {Curve {keys [<at 0 0>,<at 0.4237 0.934>,<at 1 0>]}}}
     Scale Curve=&Multiply_Float(A=0.5, B=&FN_FrontEndFloatModifier(Gameplay=1, Front End=0.5))
    )
    ScaleA=true
    ScaleRGB=true
   )
  ]
  rendererClasses [&NiagaraSpriteRendererProperties]
  renderers [{rendererClass=&NiagaraSpriteRendererProperties}]
 }
}
summary {
 systemName=NS_Emote_VastGeese_Beam1
 emitters [
  {emitterName=ConeLights, rendererClasses [&NiagaraLightRendererProperties, &NiagaraMeshRendererProperties]}
  {emitterName=DeployConeLight, rendererClasses [&NiagaraMeshRendererProperties]}
  {emitterName=Rim_Glow, rendererClasses [&NiagaraSpriteRendererProperties]}
  {emitterName=ConeLights001, rendererClasses [&NiagaraMeshRendererProperties]}
  {bEnabled=false, emitterName=Light_Down, rendererClasses [&NiagaraComponentRendererProperties]}
 ]
 userVariables [
  User.bIsFrontEnd {type=NiagaraBool}
  User.bIsFrontEndPreview {type=NiagaraBool}
  User.Flame_Size {type=NiagaraFloat, defaultValue=1, description=""}
  User.OwnerSkeleton {
   type=NiagaraDataInterfaceSkeletalMesh
   defaultValue {
    struct=NiagaraExt_VariableValue_DataInterface
    value {dataInterface=&self:NiagaraDataInterfaceSkeletalMesh_0}
   }
  }
 ]
}
