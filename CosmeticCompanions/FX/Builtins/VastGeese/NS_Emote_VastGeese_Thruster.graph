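# Niagara system for the rocket thruster effect in the VastGeese cosmetic companion emote.
# Comet_Flame_Mesh continuously spawns short-lived mesh flames at the skeletal mesh's rocketSocket, shifting colour, fading, and scaling by curves and User.Flame_Size.
# ConeLights bursts a long-lived light that follows a sampled socket, with alpha flicker keyed to system age.
# ConeLights retains beam-style mesh scaling and time-of-day material parameters despite having only a light renderer.
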NiagaraSystem /CosmeticCompanions/FX/Builtins/VastGeese/NS_Emote_VastGeese_Thruster
dependencies {
 usedDataInterfaces [&NiagaraDataInterfaceColorCurve, &NiagaraDataInterfaceCurve]
 usedDynamicInputs []
 usedModules [
  &self:SystemSpawnScript.SystemScriptSource.SystemScriptGraph.NiagaraNodeAssignment_0.SetVariables_#1
  &SystemState
 ]
}
emitters {
 Comet_Flame_Mesh {
  emitterUpdateScript [
   &EmitterState(Life Cycle Mode=::NewEnumerator1, Loop Behavior=::NewEnumerator1, Loop Delay=0.2, Loop Duration=8, UseLoopDelay=false)
   &SpawnRate(SpawnRate=20, Use Spawn Probability=false)
  ]
  particleSpawnScript [
   &InitializeParticle(
    Color=<rgba 1.5 0.375 0.15 5>
    Lifetime=0.2
    Mass Mode=::NewEnumerator1
    Mesh Scale=<-0.04 0.035 0.035>
    Mesh Scale Mode=::NewEnumerator4
    UsePositionOffset=false
    Write Material Random=true
   ) {module=InitializeParticle001}
   &InitialMeshOrientation(Mesh Orientation Mode=::NewEnumerator2, Orientation Vector=<0 0 -1>, Use Rotation Vector=false)
   /SetVariables(Particles.Color_2:LinearColor=<rgba 20 12.5 2 0.25>, Particles.Scale_2:Vector3f=<0.035 0.025 0.025>) {isSetParameters=true}
   &DynamicMaterialParameters(
    Index 0 Param 1=5
    Index 0 Param 2=5
    Module.Index0Float0Write=true
    Module.Index0Float1Write=true
    Module.Index0Float2Write=true
    Module.Index0Float3Write=true
   )
  ]
  particleUpdateScript [
   &SkeletalMeshLocation(
    MaskSpawnGroup=false
    Mesh Sampling Type=::NewEnumerator5
    Skeletal Mesh=DataInterface() {propertyValues {
      DefaultMesh=null
      ExcludeBoneName=root
      OptionalComponentClass=null
      PreviewMesh=null
      SkinningMode=SkinOnTheFly
      SoftSourceActor=null
      SourceMode=Default
      UvSetIndex=0
      WholeMeshLOD=-1
      bExcludeBone=false
      bReadDeformedGeometry=true
      bRequireCurrentFrameData=true
      ComponentTags []
      FilteredBones []
      FilteredSockets [rocketSocket]
      MeshUserParameter {parameter {name=None, typeDefHandle {registeredTypeIndex=1}, varData []}}
      SamplingRegions []
     }}
    Socket Sampling Mode=::NewEnumerator5
   )
   &ParticleState()
   &ScaleColor(
    Scale Alpha=&Lerp_Float(A=1, B=0, Alpha=$Particles.NormalizedAge)
    Scale RGB=&MakeVectorFromLinearColorRGB(Color=&ColorFromCurve(ColorCurve=DataInterface() {propertyValues {
        AlphaCurve {keys []}
        BlueCurve {keys [<at 0 0.25 RCIM_Cubic>,<at 0.492 0.5>,<at 1 0.875 RCIM_Cubic>]}
        GreenCurve {keys [<at 0 0.4375 RCIM_Cubic>,<at 0.492 0.625>,<at 1 0.9062 RCIM_Cubic>]}
        RedCurve {keys [<at 0 1 RCIM_Cubic>,<at 0.492 1>,<at 1 1 RCIM_Cubic>]}
       }}))
    ScaleA=true
    ScaleRGB=true
   )
   &ScaleMeshSize(Scale Factor=&MakeVector(
     X=&Multiply_Float(
      A=$User.Flame_Size
      B=&FloatFromCurve(FloatCurve=DataInterface() {propertyValues {Curve {keys [<at 0 0.75 RCIM_Cubic>, {arriveTangent=-0.25, interpMode=RCIM_Cubic, leaveTangent=-0.25, time=0.5, value=2.5}, <at 1 0.5 RCIM_Cubic>]}}})
     )
     Y=&Multiply_Float(
      A=$User.Flame_Size
      B=&FloatFromCurve(FloatCurve=DataInterface() {propertyValues {Curve {keys [<at 0.5 1 RCIM_Cubic>,<at 1 1.75 RCIM_Cubic>]}}})
     )
     Z=&Multiply_Float(
      A=$User.Flame_Size
      B=&FloatFromCurve(FloatCurve=DataInterface() {propertyValues {Curve {keys [<at 0.5 1 RCIM_Cubic>,<at 1 1.75 RCIM_Cubic>]}}})
     )
    ))
   /SetVariables(Particles.Scale_2:Vector3f=&Multiply_Vector(
     A=$Particles.Initial.Scale_2
     B=&MakeVector(
      X=&FloatFromCurve(
       FloatCurve=DataInterface() {propertyValues {Curve {keys [<at 0 0.75 RCIM_Cubic>, {arriveTangent=-0.25, interpMode=RCIM_Cubic, leaveTangent=-0.25, time=0.5, value=2.5}, <at 1 0.5 RCIM_Cubic>]}}}
       Scale Curve=$User.Flame_Size
      )
      Y=&FloatFromCurve(FloatCurve=DataInterface() {propertyValues {Curve {keys [<at 0.5 1 RCIM_Cubic>,<at 1 1.75 RCIM_Cubic>]}}}, Scale Curve=$User.Flame_Size)
      Z=&FloatFromCurve(FloatCurve=DataInterface() {propertyValues {Curve {keys [<at 0.5 1 RCIM_Cubic>,<at 1 1.75 RCIM_Cubic>]}}}, Scale Curve=$User.Flame_Size)
     )
    )) {isSetParameters=true}
  ]
  rendererClasses [&NiagaraMeshRendererProperties]
  renderers [
   {rendererClass=&NiagaraMeshRendererProperties, rendererIndex=0}
   {rendererClass=&NiagaraMeshRendererProperties, rendererIndex=1}
  ]
 }
 ConeLights {
  emitterUpdateScript [
   &EmitterState(Life Cycle Mode=::NewEnumerator1, Loop Behavior=::NewEnumerator1, Loop Delay=1.7, Loop Duration=8, UseLoopDelay=false)
   &SpawnBurst_Instantaneous(Use Spawn Probability=false, UseLoopCountLimit=false)
  ]
  particleSpawnScript [
   &InitializeParticle(
    Color=<rgba 1 0.2868 0 0.05>
    Lifetime=8
    Mesh Scale=<0.6 0.6 0.6>
    Mesh Scale Mode=::NewEnumerator4
    Sprite Size Mode=::NewEnumerator3
    Uniform Sprite Size=1000
    UsePositionOffset=false
    Write Material Random=true
   )
   &InitialMeshOrientation(Mesh Orientation Mode=::NewEnumerator2, Use Rotation Vector=true)
  ]
  particleUpdateScript [
   &SolveForcesAndVelocity(Clamp Velocity=false, Limit Acceleration=false)
   &ParticleState()
   &SkeletalMeshLocation(
    MaskSpawnGroup=false
    Mesh Sampling Type=::NewEnumerator5
    Orientation Sampling=::NewEnumerator1
    Skeletal Mesh=$System.Skeletal Mesh
    Socket Sampling Mode=::NewEnumerator5
   ) {module=SkeletalMeshLocation001}
   &ScaleMeshSize(Scale Factor=&Multiply_Vector(
     A=&MakeVector(
      X=&FloatFromCurve(FloatCurve=DataInterface() {propertyValues {Curve {keys [
           <at 0 0 RCIM_Cubic>
           {arriveTangentWeight=0.0247, interpMode=RCIM_Cubic, leaveTangentWeight=0.0255, tangentMode=RCTM_User, time=0.0174, value=0.05}
           {arriveTangentWeight=0.0589, interpMode=RCIM_Cubic, leaveTangentWeight=0.2882, tangentMode=RCTM_User, time=0.0565, value=0.05}
           {
            arriveTangent=3.9696
            arriveTangentWeight=0.0453
            interpMode=RCIM_Cubic
            leaveTangent=3.9696
            leaveTangentWeight=0.4261
            tangentMode=RCTM_User
            time=0.1319
            value=0.75
           }
           {arriveTangent=0.288, interpMode=RCIM_Cubic, leaveTangent=0.288, time=0.2388, value=0.8}
           {arriveTangentWeight=0.3013, interpMode=RCIM_Cubic, tangentMode=RCTM_User, time=1, value=1}
          ]}}})
      Y=&FloatFromCurve(FloatCurve=DataInterface() {propertyValues {Curve {keys [
           <at 0 0 RCIM_Cubic>
           {arriveTangentWeight=0.0247, interpMode=RCIM_Cubic, leaveTangentWeight=0.0255, tangentMode=RCTM_User, time=0.0174, value=0.05}
           {arriveTangentWeight=0.0589, interpMode=RCIM_Cubic, leaveTangentWeight=0.2882, tangentMode=RCTM_User, time=0.0565, value=0.05}
           {
            arriveTangent=3.9696
            arriveTangentWeight=0.0453
            interpMode=RCIM_Cubic
            leaveTangent=3.9696
            leaveTangentWeight=0.4261
            tangentMode=RCTM_User
            time=0.1319
            value=0.75
           }
           {arriveTangent=0.288, interpMode=RCIM_Cubic, leaveTangent=0.288, time=0.2254, value=0.8}
           {arriveTangentWeight=0.3013, interpMode=RCIM_Cubic, tangentMode=RCTM_User, time=1, value=1}
          ]}}})
      Z=&Multiply_Float(
       A=&FloatFromCurve(FloatCurve=DataInterface() {propertyValues {Curve {keys [
            {arriveTangent=779.4893, interpMode=RCIM_Cubic, leaveTangent=779.4893, leaveTangentWeight=3.3909, tangentMode=RCTM_User, time=0, value=0}
            {
             arriveTangent=19.032
             arriveTangentWeight=2.8007
             interpMode=RCIM_Cubic
             leaveTangent=19.032
             leaveTangentWeight=13.8693
             tangentMode=RCTM_User
             time=0.099
             value=30
            }
            {arriveTangentWeight=0.2773, interpMode=RCIM_Cubic, leaveTangentWeight=0.3745, tangentMode=RCTM_User, time=1.2045, value=45}
           ]}}})
       B=&RemapRange(
        Input Value To Remap=&VectorLength(Vector=&Subtract_Position(
          A=Linked() {linkedVariable {name=Particles.SkeletalMeshLocation001.SampledScale, type=Vector3f}}
          B=Linked() {linkedVariable {name=Particles.SkeletalMeshLocation002.SampledScale, type=Vector3f}}
         ))
        Input Min=0
        Input Max=100
        Output Min=0
        Output Max=2.5
       )
      )
     )
     B=&Divide_Vector(A:Vector3f=<1 1 1>, B:Vector3f=$Engine.Owner.Scale)
    ))
   &ScaleColor(
    Scale Alpha=&FloatFromCurve(
     FloatCurve=DataInterface() {propertyValues {Curve {keys [
         {interpMode=RCIM_Cubic, leaveTangent=0.0402, tangentMode=RCTM_Break, time=0, value=0}
         <at 0.3 0.0075>
         {arriveTangent=-0.3583, leaveTangent=-0.3583, time=3.9058, value=0.0075}
         <at 3.9074 0.0005>
         <at 3.9555 0.0005>
         <at 3.9764 0.0075>
         <at 4.1455 0.0075>
         <at 4.1461 0.0005>
         <at 4.2511 0.0005>
         <at 4.2699 0.0075>
         <at 4.3362 0.0075>
         <at 4.3431 0.0005>
         <at 4.4684 0.0005>
         <at 4.497 0.0075>
        ]}}}
     CurveIndex=$System.Age
    )
    ScaleA=true
    ScaleRGB=true
   )
   &FN_Light_Attributes(
    Color=$Particles.Color
    Create Color Attribute=true
    Create Exponent Attribute=true
    Create Position Attribute=true
    Create Radius Attribute=true
    Exponent=4
    Position=&AddVectorToPosition(Position=$Particles.SkeletalMeshLocation001.SampledPosition, Vector=<0 0 0>)
    Radius=100
    Scale Color=&FloatFromCurve(FloatCurve=DataInterface() {propertyValues {Curve {keys [<at 0 0>,<at 0.02 1>,<at 1 1>]}}}, Scale Curve=10)
   )
   /SetVariables(Particles.DynamicMaterialParameterTechBeam:Vector4f=/Niagara/DynamicInputs/Multiply/Multiply_Vector4_ByFloat(
     Vector4:Vector4f=<0.8 0.8 1 1>
     Float:NiagaraFloat=&FloatFromCurve(
      FloatCurve=DataInterface() {propertyValues {Curve {keys [
          {interpMode=RCIM_Cubic, leaveTangent=1.6164, tangentMode=RCTM_Break, time=0, value=0}
          {arriveTangentWeight=0.173, interpMode=RCIM_Cubic, leaveTangentWeight=1.0013, tangentMode=RCTM_User, time=0.4503, value=1}
          {arriveTangent=-0.0121, leaveTangent=-0.3583, tangentMode=RCTM_User, time=4.0562, value=1}
          <at 4.0574 0.25>
          <at 4.1055 0.25>
          <at 4.1267 1>
          <at 4.2958 1>
          <at 4.2961 0.25>
          <at 4.4011 0.25>
          <at 4.4202 1>
          <at 4.4865 1>
          <at 4.4931 0.25>
          <at 4.6184 0.25>
          <at 4.6473 1>
         ]}}}
      CurveIndex=$Engine.System.Age
     )
    )) {isSetParameters=true}
   &DynamicMaterialParameters(
    Index 0 Param 4=&FloatFromCurve(
     FloatCurve=DataInterface() {propertyValues {Curve {keys [
         {interpMode=RCIM_Cubic, leaveTangent=1.6164, tangentMode=RCTM_Break, time=0, value=0}
         {arriveTangentWeight=0.173, interpMode=RCIM_Cubic, leaveTangentWeight=1.0013, tangentMode=RCTM_User, time=0.4503, value=1}
         {arriveTangent=-0.0121, leaveTangent=-0.3583, tangentMode=RCTM_User, time=4.0562, value=1}
         <at 4.0574 0.25>
         <at 4.1055 0.25>
         <at 4.1267 1>
         <at 4.2958 1>
         <at 4.2961 0.25>
         <at 4.4011 0.25>
         <at 4.4202 1>
         <at 4.4865 1>
         <at 4.4931 0.25>
         <at 4.6184 0.25>
         <at 4.6473 1>
        ]}}}
     CurveIndex=$Engine.System.Age
    )
    Index 0 Vector 1:Vector3f=&FN_Select_Vector_from_Time_Of_Day(
     Morning=&MakeVectorFromLinearColorRGB(Color=<rgba 3 2.4842 0.7221 1>)
     Day=&MakeVectorFromLinearColorRGB(Color=<rgba 3 2.4842 0.7221 1>)
     Evening=&MakeVectorFromLinearColorRGB(Color=<rgba 3 2.4842 0.7221 1>)
     Night=&MakeVectorFromLinearColorRGB(Color=<rgba 5 2.6977 0.5729 1>)
    )
    Module.Index0Float3Write=true
    Module.Index0VectorWrite=true
    Write Parameter Index 0=::NewEnumerator3
   )
  ]
  rendererClasses [&NiagaraLightRendererProperties]
  renderers [{rendererClass=&NiagaraLightRendererProperties, rendererIndex=0}]
 }
}
summary {
 systemName=NS_Emote_VastGeese_Thruster
 emitters [
  {emitterName=Comet_Flame_Mesh, rendererClasses [&NiagaraMeshRendererProperties]}
  {emitterName=ConeLights, rendererClasses [&NiagaraLightRendererProperties]}
 ]
 userVariables [
  User.Flame_Size {type=NiagaraFloat, defaultValue=1, description=""}
  User.OwnerSkeleton {
   type=NiagaraDataInterfaceSkeletalMesh
   defaultValue {
    struct=NiagaraExt_VariableValue_DataInterface
    value {dataInterface=&self:NiagaraDataInterfaceSkeletalMesh_0}
   }
  }
 ]
}
